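# Niagara system for the idle effect of the Hightower umbrella glider: red electricity, embers, trails and glow.
# GPU mesh/sprite arcs follow a canopy-like dome by setting height from horizontal distance, alongside short SubUV spark sprites.
# Embers spawn over time and with owner movement, driven by curl noise and wind.
# Skeletal mesh sockets anchor glow sprites and lights, while two wind-swept ribbon AnimTrails adjust their motion for front end versus gameplay.
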NiagaraSystem /BRCosmetics/Gadgets/Assets/VinderTech_Umbrella/Umbrella_Hightower/FX/NS_HighTower_Umbrella_Idle
dependencies {
 usedDataInterfaces [&NiagaraDataInterfaceCurve, &NiagaraDataInterfaceCamera, &NiagaraDataInterfaceVectorCurve]
 usedDynamicInputs []
 usedModules [&SystemState]
}
emitters {
 ElectricArcs {
  simTarget=GPUComputeSim
  emitterUpdateScript [
   &EmitterState()
   &SpawnRate(Spawn Probability=0.4, SpawnRate=10, Use Spawn Probability=true)
  ]
  particleSpawnScript [
   /Niagara/Modules/Spawn/Initialization/InitializeParticle(
    Color=<rgba 10 0.4 0.4 1>
    Lifetime=&UniformRangedFloat(Minimum=0.25, Maximum=0.6)
    Mesh Scale=&UniformRangedVector(Minimum=<0.5 1 1>, Maximum=<1.5 2 3>)
    Module.Write Color=true
    Module.Write Lifetime=true
    Module.Write Mass=true
    Module.Write Material Random=true
    Module.Write Position=false
    Module.Write Scale=true
    Module.Write SpriteRotation=false
    Module.Write SpriteSize=true
    Position=&SimulationPosition()
    Sprite Size=&Vector2DFromFloat(Value=&UniformRangedFloat(Minimum=35, Maximum=45))
   )
   &BoxLocation(
    Box Offset=<0 -10 0>
    Box Origin:NiagaraPosition=/Niagara/DynamicInputs/Helpers/SimulationPosition()
    Box Size=<90 15 2>
    MaskSpawnGroup=false
    Module.Override Seed=false
    Rotate Box=false
    Surface Only=false
   )
   &AddVelocityFromPoint(Module.Use Falloff Distance=false, Velocity Strength=&UniformRangedFloat(Minimum=60, Maximum=90))
   &CameraOffset(Camera Offset Amount=7)
   /SetVariables(Particles.CameraOffset2=22) {isSetParameters=true}
   &InitialMeshOrientation(Orientation Vector=<-1 0 0>, Use Orientation Vector=true, Use Rotation Vector=false) {module=InitialMeshOrientation001}
   &DynamicMaterialParameters(
    Index 0 Param 1=&UniformRangedFloat(Minimum=4, Maximum=4)
    Index 0 Param 3=$Particles.MaterialRandom
    Index 0 Param 4=&UniformRangedFloat(Minimum=0.5, Maximum=0.8)
    Module.Param0WriteEnabled=true
    Module.Param1WriteEnabled=false
    Module.Param2WriteEnabled=true
    Module.Param3WriteEnabled=true
   ) {module=DynamicMaterialParameters001}
   /SetVariables(Particles.SpriteSize2=&Vector2DFromFloat(Value=&UniformRangedFloat(Minimum=7, Maximum=11))) {isSetParameters=true}
  ]
  particleUpdateScript [
   &ParticleState()
   /Niagara/Modules/Update/Orientation/V2/OrientMeshToVector(
    Look At Direction:Vector3f=&Multiply_Vector(
     A=<1 1 0>
     B=&Subtract_Vector(A=Linked() {linkedVariable {name=Particles.Position, type=NiagaraPosition}}, B=&SimulationPosition())
    )
    Override Quat=false
    UseOptionalSideAxis=false
    UseOptionalUpAxis=false
   )
   &SolveForcesAndVelocity(Clamp Velocity=false, Limit Acceleration=false)
   &ScaleColor(
    Scale Alpha=&self:Flicker(
     FlickerAmount:NiagaraFloat=&FloatFromCurve(
      FloatCurve=DataInterface() {propertyValues {Curve {keys [
          {arriveTangent=2.7223, interpMode=RCIM_Cubic, leaveTangent=2.7223, tangentMode=RCTM_User, time=0, value=1}
          {arriveTangent=0.0266, interpMode=RCIM_Cubic, leaveTangent=0.0266, time=0.2457, value=0.5}
          {arriveTangent=-0.3314, interpMode=RCIM_Cubic, leaveTangent=-0.3314, time=0.527, value=1.014}
          {arriveTangent=-3.2376, interpMode=RCIM_Cubic, leaveTangent=-11960.9277, tangentMode=RCTM_User, time=1, value=0.25}
         ]}}}
      CurveIndex=$Particles.Age
     )
     FlickerInterval:NiagaraFloat=0.1
    )
    ScaleA=true
    ScaleRGB=true
   )
   /SetVariables(Particles.Color2=&MultiplyLinearColor(
     Linear Color=$Particles.Initial.Color
     Scale Factor=&MakeFloat4FromVectorAndFloat(
      "Vector (RGB)"=<0.1 0.07 0.02>
      "Float (Alpha)"=&FloatFromCurve(
       FloatCurve=DataInterface() {propertyValues {Curve {keys [
           <at 0 0 RCIM_Cubic>
           {interpMode=RCIM_Cubic, tangentMode=RCTM_User, time=0.2, value=1}
           {interpMode=RCIM_Cubic, tangentMode=RCTM_User, time=0.6, value=1}
           <at 1 0 RCIM_Cubic>
          ]}}}
       Scale Curve=0.7
      )
     )
    )) {isSetParameters=true}
   /SetVariables(Particles.Position=&MakeVector(
     X=&MakeFloatFromVector(Vector=Linked() {linkedVariable {name=Particles.Position, type=NiagaraPosition}}, Channel=::NewEnumerator0)
     Y=&MakeFloatFromVector(Vector=Linked() {linkedVariable {name=Particles.Position, type=NiagaraPosition}}, Channel=::NewEnumerator1)
     Z=&FloatFromCurve(
      FloatCurve=DataInterface() {propertyValues {Curve {keys [
          <at 0 7 RCIM_Cubic>
          {arriveTangent=-0.8879, interpMode=RCIM_Cubic, leaveTangent=-0.8879, tangentMode=RCTM_User, time=80, value=-30}
          <at 100 -80>
         ]}}}
      CurveIndex=&self:FlatDistanceLength()
     )
    )) {isSetParameters=true}
   /SetVariables(Particles.DynamicParams2:Vector4f=&MakeVector4(
     X=&Lerp_Float(
      A=2
      B=0.25
      Alpha=&ClampFloat(Float=&DotProduct(Vector A=&ReturnCameraVector(Camera Property=::NewEnumerator1, Coordinate Space=::World), Vector B=<0 0 -1>))
     )
     Y=1.5
     Z=&Multiply_Float(
      A=&Multiply_Float(A=/Niagara/DynamicInputs/TypeConversions/VectorLengthSafe(Vector:Vector3f=$Particles.Initial.Velocity), B=0.02)
      B=&Add_Float(A=$Particles.MaterialRandom, B=$Particles.NormalizedAge)
     )
     W=0
    )) {isSetParameters=true}
   &DynamicMaterialParameters(
    Index 0 Param 1=&FloatFromCurve(
     FloatCurve=DataInterface() {propertyValues {Curve {keys [<at 0.5 1 RCIM_Cubic>, {arriveTangent=-2.5142, interpMode=RCIM_Cubic, leaveTangent=-2.5142, tangentMode=RCTM_User, time=1, value=0}]}}}
     Scale Curve=&Lerp_Float(A=0.7, B=1.5, Alpha=$Particles.Initial.MaterialRandom)
    )
    Index 0 Param 2=&FloatFromCurve(
     FloatCurve=DataInterface() {propertyValues {Curve {keys [
         {arriveTangent=10.9494, interpMode=RCIM_Cubic, leaveTangent=10.9494, tangentMode=RCTM_User, time=0, value=1}
         {arriveTangent=-8.5345, interpMode=RCIM_Cubic, leaveTangent=-8.5345, tangentMode=RCTM_User, time=0.2, value=2}
         {arriveTangent=-2.4364, interpMode=RCIM_Cubic, leaveTangent=-2.4364, time=0.3638, value=0.5858}
         {arriveTangent=-8.496, interpMode=RCIM_Cubic, leaveTangent=-8.496, tangentMode=RCTM_User, time=0.4639, value=1.357}
         {arriveTangent=8.1076, interpMode=RCIM_Cubic, leaveTangent=-1.1376, tangentMode=RCTM_User, time=1, value=1.5}
        ]}}}
     Scale Curve=0.75
    )
    Module.Param0WriteEnabled=true
    Module.Param1WriteEnabled=true
    Module.Param2WriteEnabled=false
    Module.Param3WriteEnabled=false
   )
   &ScaleMeshSize(Scale Factor=&VectorFromCurve(
     VectorCurve=DataInterface() {propertyValues {
       XCurve {keys [
         {arriveTangent=2.9222, interpMode=RCIM_Cubic, leaveTangent=2.9222, tangentMode=RCTM_User, time=0, value=0}
         {interpMode=RCIM_Cubic, tangentMode=RCTM_User, time=0.65, value=1}
        ]}
       YCurve {keys [<at 0 1>]}
       ZCurve {keys [
         {arriveTangent=3.2926, interpMode=RCIM_Cubic, leaveTangent=3.2925, tangentMode=RCTM_User, time=0, value=0.1}
         {arriveTangent=-1.316, interpMode=RCIM_Cubic, leaveTangent=-1.316, tangentMode=RCTM_User, time=0.2588, value=1}
         {arriveTangent=0.348, interpMode=RCIM_Cubic, leaveTangent=0.348, tangentMode=RCTM_User, time=0.4597, value=0.4069}
         {interpMode=RCIM_Cubic, leaveTangent=3.5341, time=1, value=1.5}
        ]}
      }}
     CurveIndex=$Particles.Age
    ))
   /SetVariables(Particles.DynamicParamOverride:Vector4f=&MakeFloat4FromVectorAndFloat(
     "Vector (RGB)"=/Niagara/DynamicInputs/Transforms/TransformPosition(
      Position=Linked() {linkedVariable {name=Particles.Initial.Position, type=NiagaraPosition}}
      Source Space=::Simulation
      Destination Space=::World
     )
     "Float (Alpha)"=0
    )) {isSetParameters=true}
   /NiagaraFN/Sprite/FN_SpriteAlignmentAndFacing(SpriteFacing:Vector3f=&Subtract_Vector(A=Linked() {linkedVariable {name=Particles.Position, type=NiagaraPosition}}, B=<0 0 -90>)) {module=SpriteAlignmentAndFacing}
  ]
  rendererClasses [&NiagaraMeshRendererProperties, &NiagaraSpriteRendererProperties]
  renderers [
   {rendererClass=&NiagaraMeshRendererProperties, rendererIndex=0}
   {rendererClass=&NiagaraSpriteRendererProperties, rendererIndex=1}
   {rendererClass=&NiagaraSpriteRendererProperties, rendererIndex=2}
  ]
 }
 ElectricSprites {
  emitterUpdateScript [
   &EmitterState(
    Inactive Response=::NewEnumerator1
    Life Cycle Mode=::NewEnumerator1
    Loop Duration=&UniformRangedFloat(Minimum=0.5, Maximum=1.8)
    Recalculate Duration Each Loop=true
    UseLoopDelay=false
   )
   &SpawnRate(Spawn Probability=0.5, SpawnRate=12, Use Spawn Probability=true)
   &SpawnBurst_Instantaneous(Spawn Count=&UniformRangedInt(Minimum=0, Maximum=5), Use Spawn Probability=false)
  ]
  particleSpawnScript [
   &InitializeParticle(
    Color=<rgba 12 0.7 0.7 1>
    Lifetime=&UniformRangedFloat(Minimum=0.1, Maximum=0.17)
    Module.Write Color=true
    Module.Write Lifetime=true
    Module.Write Mass=true
    Module.Write Material Random=true
    Module.Write Position=false
    Module.Write Scale=false
    Module.Write SpriteRotation=true
    Module.Write SpriteSize=true
    Position=/Niagara/Functions/Localspace/SimulationPosition()
    Sprite Rotation=&UniformRangedFloat(Minimum=0, Maximum=360)
    Sprite Size=&UniformRangedVector2D(Minimum=<35 35>, Maximum=<55 55>)
   )
   /Niagara/Modules/Spawn/Location/SphereLocation(
    Hemisphere Z=true
    MaskSpawnGroup=false
    Module.Override Randomness=false
    Non Uniform Scale=<1 0.4 0.45>
    Offset=<0 0 -30>
    Sphere Radius=&FloatFromCurve(FloatCurve=DataInterface() {propertyValues {Curve {keys [<at 0 8>,<at 0.2 20>,<at 0.25 80>]}}}, CurveIndex=$Emitter.Age)
    Surface Only=true
   )
   /Niagara/Modules/Update/SubUV/SubUVAnimation(SubUV Animation Mode=::NewEnumerator1)
   &DynamicMaterialParameters(
    Index 0 Param 1=$Particles.MaterialRandom
    Index 0 Param 2=&Lerp_Float(A=2, B=4, Alpha=$Particles.MaterialRandom)
    Module.Param0WriteEnabled=true
    Module.Param1WriteEnabled=true
    Module.Param2WriteEnabled=false
    Module.Param3WriteEnabled=false
   )
   &CameraOffset(Camera Offset Amount=20)
  ]
  particleUpdateScript [
   &ParticleState()
   &DynamicMaterialParameters(
    Index 0 Param 3=&FloatFromCurve(FloatCurve=DataInterface() {propertyValues {Curve {keys [<at 0.2 0 RCIM_Cubic>, {interpMode=RCIM_Cubic, leaveTangent=1.5819, time=1, value=1}]}}})
    Module.Param0WriteEnabled=false
    Module.Param1WriteEnabled=false
    Module.Param2WriteEnabled=true
    Module.Param3WriteEnabled=false
   ) {module=DynamicMaterialParameters001}
   &ScaleSpriteSize(Scale Factor=&Vector2DFromFloat(Value=&FloatFromCurve(FloatCurve=DataInterface() {propertyValues {Curve {keys [{arriveTangent=0.3147, interpMode=RCIM_Cubic, leaveTangent=0.3147, tangentMode=RCTM_User, time=0, value=0.8}, <at 1 1>]}}})))
  ]
  rendererClasses [&NiagaraSpriteRendererProperties]
  renderers [{rendererClass=&NiagaraSpriteRendererProperties}]
 }
 Embers {
  emitterUpdateScript [
   &EmitterScalability(
    Module.bApplySpawnCountScale=false
    Module.bApplySpawnCountScaleByDistanceFraction=false
    Module.bUseMaxDistance=false
    Module.bUseMinDistance=false
   ) {enabled=false}
   &EmitterLifeCycle() {enabled=false}
   &EmitterState(
    Life Cycle Mode=::NewEnumerator1
    Loop Duration=&UniformRangedFloat(Minimum=0.5, Maximum=1.8)
    Recalculate Duration Each Loop=true
    UseLoopDelay=false
   )
   &SpawnBurst_Instantaneous(Spawn Count=&UniformRangedInt(Minimum=0, Maximum=14), Use Spawn Probability=false)
   &SpawnRate(SpawnRate=20, Use Spawn Probability=false)
   &SpawnPerUnit(
    Max Movement Threshold=3000
    Module.Use Max Movement Threshold=true
    Module.Use Movement Tolerance=false
    Spawn Group=1
    Spawn Per Unit=20
    Spawn Probability=0.7
    Use Spawn Probability=true
    Velocity Vector:Vector3f=&Multiply_Vector(A=$Engine.Owner.Velocity, B=<1 1 0.25>)
   )
  ]
  particleSpawnScript [
   &InitializeParticle(
    Color=<rgba 6 2 3 1>
    Lifetime=&UniformRangedFloat(Minimum=0.14, Maximum=0.5)
    Module.Write Color=true
    Module.Write Lifetime=true
    Module.Write Mass=true
    Module.Write Material Random=true
    Module.Write Position=true
    Module.Write Scale=false
    Module.Write SpriteRotation=false
    Module.Write SpriteSize=true
    Position=&MakeVector(
     X=&FloatFromCurve(
      FloatCurve=DataInterface() {propertyValues {Curve {keys [
          {interpMode=RCIM_Cubic, tangentMode=RCTM_User, time=0, value=1}
          {arriveTangent=-4.512, interpMode=RCIM_Cubic, tangentMode=RCTM_User, time=0.49, value=0.2}
          {arriveTangent=-4.3496, interpMode=RCIM_Cubic, leaveTangent=-4.3496, tangentMode=RCTM_User, time=0.5, value=-0.2}
          {interpMode=RCIM_Cubic, tangentMode=RCTM_User, time=1, value=-1}
         ]}}}
      CurveIndex=&UniformRangedFloat(Minimum=0, Maximum=1)
      Scale Curve=70
     )
     Y=&UniformRangedFloat(Minimum=-10, Maximum=10)
     Z=0
    )
    Sprite Size=&Vector2DFromFloat(Value=&FloatFromCurve(
      FloatCurve=DataInterface() {propertyValues {Curve {keys [<at 0 1>,<at 0.2 0.6>,<at 1 0.4>]}}}
      CurveIndex=&UniformRangedFloat(Minimum=500, Maximum=1000)
      Scale Curve=5
     ))
   )
   &AddVelocityInCone(Module.Use Velocity Falloff On Cone Axis=false) {enabled=false}
   /SetVariables(Particles.Position=&Lerp_Vector(
     A=&MakeVector(
      X=&MakeFloatFromVector(Vector=Linked() {linkedVariable {name=Particles.Position, type=NiagaraPosition}}, Channel=::NewEnumerator0)
      Y=&MakeFloatFromVector(Vector=Linked() {linkedVariable {name=Particles.Position, type=NiagaraPosition}}, Channel=::NewEnumerator1)
      Z=&FloatFromCurve(
       FloatCurve=DataInterface() {propertyValues {Curve {keys [<at 0 8 RCIM_Cubic>, {arriveTangent=-0.8879, interpMode=RCIM_Cubic, leaveTangent=-0.8879, tangentMode=RCTM_User, time=80, value=-33}]}}}
       CurveIndex=&self:FlatDistanceLength()
      )
     )
     B=&MakeVector(
      X=&UniformAOrBFloat(A=-72, B=72)
      Y=&UniformAOrBFloat(A=-14, B=14)
      Z=-22
     )
     Alpha=&MakeFloatFromInt(Int=&SpawnGroup())
    )) {isSetParameters=true}
   /Niagara/Modules/Spawn/Location/V2/SphereLocation(
    MaskSpawnGroup=false
    Offset=<0 0 30>
    Override Seed=false
    Sphere Radius=0
    Surface Only=false
    UseNonUniformScale=false
    UseOffset=true
    UseSphereOrientation=false
   ) {enabled=false}
   &AddVelocityFromPoint(
    Module.Use Falloff Distance=false
    Origin Offset=&MakeVector(
     X=0
     Y=0
     Z=&Multiply_Float(A=40, B=&MakeFloatFromInt(Int=&SpawnGroup()))
    )
    Velocity Strength=&UniformRangedFloat(Minimum=50, Maximum=250)
   )
   &InheritVelocity(
    Module.Limit Speed=true
    Velocity Limit=400
    Velocity Scale=<-1.5 -1.5 0>
    Velocity Source=&TransformVector(Vector=$Engine.Owner.Velocity, Source Space=::World, Destination Space=::Simulation)
   )
  ]
  particleUpdateScript [
   &ParticleState()
   &AccelerationForce(Acceleration=<0 0 200>) {enabled=false}
   &Drag(
    Drag=&FloatFromCurve(
     FloatCurve=DataInterface() {propertyValues {Curve {keys [<at 0 0.25 RCIM_Cubic>, {arriveTangent=0.75, interpMode=RCIM_Cubic, leaveTangent=0.75, time=0.5, value=0.4}, <at 1 1 RCIM_Cubic>]}}}
     Scale Curve=8
    )
    Use Linear Drag=true
    Use Rotational Drag=false
   )
   /Niagara/Modules/Update/Forces/CurlNoiseForce(
    Noise Frequency=0.03
    Noise Strength=&FloatFromCurve(FloatCurve=DataInterface() {propertyValues {Curve {keys [<at 0 1 RCIM_Cubic>,<at 1 0>]}}}, Scale Curve=2000)
    Pan Noise=false
    Pan Noise Field=<0 0 1>
   )
   &WindForce(
    Air Resistance=0.4
    Wind Speed=&TransformVector(Vector=$Engine.Owner.SystemYAxis, Source Space=::World, Destination Space=::Simulation)
    Wind Speed Scale=-2000
   )
   &SolveForcesAndVelocity(Clamp Velocity=false, Limit Acceleration=false)
   &Color(
    Scale Alpha=HlslExpression() {hlslExpression=1 - Particles.NormalizedAge}
    Scale Color=HlslExpression() {hlslExpression=1 - Particles.NormalizedAge}
   ) {enabled=false}
   &ScaleSpriteSizeBySpeed(
    Max Scale Factor=<0.8 10>
    Sample Scale Factor By Curve=true
    Scale Factor Curve=DataInterface() {propertyValues {Curve {keys [
        <at 0 0 RCIM_Cubic>
        {arriveTangent=0.6549, interpMode=RCIM_Cubic, leaveTangent=0.6549, tangentMode=RCTM_User, time=0.7042, value=0.3148}
        {arriveTangent=4.3749, interpMode=RCIM_Cubic, leaveTangent=4.3749, tangentMode=RCTM_User, time=1, value=1}
       ]}}}
    Velocity Threshold=600
   )
   &ScaleSpriteSize(Scale Factor=&Vector2DFromFloat(Value=&FloatFromCurve(FloatCurve=DataInterface() {propertyValues {Curve {keys [
          {arriveTangent=31.3403, interpMode=RCIM_Cubic, leaveTangent=31.3403, tangentMode=RCTM_User, time=0, value=0}
          <at 0.05 1 RCIM_Cubic>
          {arriveTangent=-2.1718, interpMode=RCIM_Cubic, leaveTangent=-2.1718, tangentMode=RCTM_User, time=1, value=0}
         ]}}})))
   &ScaleColor(
    Scale RGB=&VectorFromCurve(
     VectorCurve=DataInterface() {propertyValues {
       XCurve {keys [{arriveTangent=-0.8497, interpMode=RCIM_Cubic, leaveTangent=-0.8497, tangentMode=RCTM_User, time=0, value=1}, <at 1 0.7>]}
       YCurve {keys [{arriveTangent=-1.8823, interpMode=RCIM_Cubic, leaveTangent=-1.8823, tangentMode=RCTM_User, time=0, value=1}, <at 0.8 0.02>]}
       ZCurve {keys [{arriveTangent=-1.8429, interpMode=RCIM_Cubic, leaveTangent=-1.8429, tangentMode=RCTM_User, time=0, value=1}, <at 0.8 0.02>]}
      }}
     Scale Curve=&VectorFromFloat(Value=3)
    )
    ScaleA=false
    ScaleRGB=true
   )
  ]
  rendererClasses [&NiagaraSpriteRendererProperties]
  renderers [{rendererClass=&NiagaraSpriteRendererProperties}]
 }
 FN_AnimTrail_V4001 {
  emitterUpdateScript [
   &EmitterState()
   &FN_SpawnAnimTrail_v2(
    Number of Anim Trails=2
    Spawn Interval:NiagaraInt32=&MakeCustomIntFromBool(Bool=$User.bIsFullyDeployed, True INT=1, False INT=1)
   )
  ]
  particleSpawnScript [
   /Game/Effects/Niagara/Modules/AnimTrail/FN_InitializeAnimTrail/FN_InitializeAnimTrailBoneIndecies(First Bone or Socket Index=1, Second Bone or Socket Index=3)
   &FN_InitializeAnimTrailBoneIndecies(AnimTrail Selection=1, Second Bone or Socket Index=2) {module=FN_InitializeAnimTrailBoneIndecies001}
   /Game/Effects/Niagara/Modules/AnimTrail/FN_InitializeAnimTrail/FN_InitializeAnimTrail(AnimTrailWidth=1.25, Color=<rgba 1 0.05 0.05 0.1>, Lifetime=0.14)
   &InheritVelocity(Module.Limit Speed=true, Velocity Scale=<-0.5 -0.5 1>) {enabled=false}
   &AddVelocity(Velocity=&MakeVector(
     X=&Multiply_Float(
      A=&FN_FrontEndFloatModifier(Gameplay=1, Front End=2)
      B=&Lerp_Float(A=-400, B=400, Alpha=&MakeFloatFromInt(Int=$Particles.SpawnGroup))
     )
     Y=&self:WavyFloat(
      TimeMult:NiagaraFloat=2.5
      TimeOffset:NiagaraFloat=&MakeFloatFromInt(Int=$Particles.SpawnGroup)
      Value:NiagaraFloat=100
     )
     Z=-500
    ))
   &DynamicMaterialParameters(
    Index 0 Param 2=1
    Index 0 Param 4=5
    Module.Param0WriteEnabled=false
    Module.Param1WriteEnabled=false
    Module.Param2WriteEnabled=false
    Module.Param3WriteEnabled=false
   ) {enabled=false}
  ]
  particleUpdateScript [
   &Drag(Drag=5, Use Linear Drag=true, Use Rotational Drag=false)
   &WindForce(
    Wind Speed=$Engine.Owner.SystemYAxis
    Wind Speed Scale=&FN_FrontEndFloatModifier(Gameplay=-5000, Front End=-15000)
   )
   &SolveForcesAndVelocity(Clamp Velocity=false, Limit Acceleration=false)
   /Game/Effects/Niagara/Modules/AnimTrail/FN_UpdateAnimTrail/FN_UpdateAnimTrail()
   &ScaleColor(
    Scale Alpha=&FloatFromCurve(FloatCurve=DataInterface() {propertyValues {Curve {keys [{arriveTangent=-5.7894, interpMode=RCIM_Cubic, time=0, value=1}, <at 1 0>]}}})
    ScaleA=true
    ScaleRGB=false
   )
  ]
  rendererClasses [&NiagaraRibbonRendererProperties]
  renderers [{rendererClass=&NiagaraRibbonRendererProperties}]
 }
 Glow {
  emitterUpdateScript [
   &EmitterState(
    Inactive Response=::NewEnumerator1
    Life Cycle Mode=::NewEnumerator1
    Loop Behavior=::NewEnumerator1
    Loop Duration Mode=::NewEnumerator1
    UseLoopDelay=false
   )
   &SpawnBurst_Instantaneous(Spawn Count=2, Use Spawn Probability=false)
  ]
  particleSpawnScript [
   &FN_SetSpawnGroup_v2(Number of Spawn Groups=2)
   &InitializeParticle(
    Color=<rgba 2 0.1 0.1 1>
    Mesh Scale=<5 5 5>
    Module.Write Color=true
    Module.Write Lifetime=true
    Module.Write Mass=false
    Module.Write Material Random=true
    Module.Write Position=false
    Module.Write Scale=true
    Module.Write SpriteRotation=false
    Module.Write SpriteSize=true
    Position=&SimulationPosition()
    Sprite Size=&Vector2DFromFloat(Value=40)
   )
   /SetVariables(Particles.LightFalloff=5) {isSetParameters=true}
   &DynamicMaterialParameters(
    Index 0 Param 1=0.7
    Index 0 Param 2=0.6
    Index 1 Param 1=5
    Index 1 Param 4=1
    Module.Param0WriteEnabled=true
    Module.Param0WriteEnabled1=true
    Module.Param1WriteEnabled=true
    Module.Param1WriteEnabled1=false
    Module.Param2WriteEnabled=false
    Module.Param2WriteEnabled1=false
    Module.Param3WriteEnabled=false
    Module.Param3WriteEnabled1=false
    Write Parameter Index 1=true
   )
  ]
  particleUpdateScript [
   &ParticleState(Kill Particles When Lifetime Has Elapsed=false)
   &SkeletalMeshLocation(
    Bone / Socket Index=$Particles.SpawnGroup
    Inherit Velocity=false
    MaskSpawnGroup=false
    Mesh Sampling Type=::NewEnumerator5
    Sampled Position Coordinate Space=::World
    Skeletal Mesh=$User.SkeletalMesh
    Socket Sampling Mode=::NewEnumerator5
   )
   /SetVariables(Particles.FlareOffset:Vector3f=&Add_Vector(
     A=$Output.SkeletalMeshLocation.SampledPosition
     B=&TransformVector(Vector=<0 31 0>, Destination Space=::Simulation)
    )) {isSetParameters=true}
   &ScaleColor(
    Scale RGB=&VectorFromFloat(Value=&FloatFromCurve(FloatCurve=DataInterface() {propertyValues {Curve {keys [{arriveTangent=-28.5037, interpMode=RCIM_Cubic, time=0, value=0}, <at 1 1>]}}}))
    ScaleA=true
    ScaleRGB=true
   )
  ]
  rendererClasses [&NiagaraSpriteRendererProperties, &NiagaraLightRendererProperties]
  renderers [
   {rendererClass=&NiagaraSpriteRendererProperties}
   {rendererClass=&NiagaraLightRendererProperties, rendererIndex=1}
   {rendererClass=&NiagaraSpriteRendererProperties, rendererIndex=2}
  ]
 }
}
summary {
 systemName=NS_HighTower_Umbrella_Idle
 emitters [
  {emitterName=Glow, rendererClasses [&NiagaraSpriteRendererProperties, &NiagaraLightRendererProperties]}
  {emitterName=ElectricSprites, rendererClasses [&NiagaraSpriteRendererProperties]}
  {
   emitterName=ElectricArcs
   simTarget=GPUComputeSim
   rendererClasses [&NiagaraMeshRendererProperties, &NiagaraSpriteRendererProperties]
  }
  {emitterName=Embers, rendererClasses [&NiagaraSpriteRendererProperties]}
  {emitterName=FN_AnimTrail_V4001, rendererClasses [&NiagaraRibbonRendererProperties]}
 ]
 userVariables [
  User.bIsFrontEnd {type=NiagaraBool}
  User.bIsFrontEndPreview {type=NiagaraBool}
  User.bIsFullyDeployed {type=NiagaraBool, defaultValue=true}
  User.RibbonIDOffset {type=NiagaraInt32}
  User.SkeletalMesh {
   type=NiagaraDataInterfaceSkeletalMesh
   defaultValue {
    struct=NiagaraExt_VariableValue_DataInterface
    value {dataInterface=&self:NiagaraDataInterfaceSkeletalMesh_0}
   }
  }
 ]
}
