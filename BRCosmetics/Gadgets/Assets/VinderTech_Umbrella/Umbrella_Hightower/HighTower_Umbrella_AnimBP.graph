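# Animation Blueprint for the HighTower umbrella glider, animating its deploy, glide and closed poses with spinning parts.
# A Parachute state machine follows parachute-open and force-open flags, swapping in front-end sequences for cosmetic previews; its separate force-open state is gated by BlockAnimRule, never set here.
# Additive side and top spinners layer on top before DefaultSlot.
# Each update interpolates top-spin speed from forward acceleration, or zeroes it in preview.
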AnimBlueprint /BRCosmetics/Gadgets/Assets/VinderTech_Umbrella/Umbrella_Hightower/HighTower_Umbrella_AnimBP
target_skeleton=/BRCosmetics/Gadgets/Assets/VinderTech_Umbrella/Umbrella_Hightower/Meshes/HighTower_Skeleton
graphs [
 AnimGraph {nodes [
   AnimGraphNode_StateMachine_0 = Parachute() {is=title AnimGraphNode_StateMachine}
   n1 = Apply Additive(
    Base=@AnimGraphNode_StateMachine_0
    Additive="Sequence Player (additive) 'Umbrella_SideSpinners_GLIDER'"!(PlayRate:float=Get SideSpinner_Speed() -> (SideSpinner_Speed:double)) -> (Pose) {sequence=&/Umbrella_SideSpinners_GLIDER}
   ) {alphaBoolBlend {blendOption=Linear}}
   n2 = "Sequence Player (additive) 'Umbrella_TopSpin_GLIDER'"!(PlayRate:float=Get TopSpin_Speed() -> (TopSpin_Speed:double)) -> (Pose) {sequence=&/Umbrella_TopSpin_GLIDER}
   Output Pose(Result=Slot 'DefaultSlot'(Source=Apply Additive(Base=@n1, Additive=@n2) {alphaBoolBlend {blendOption=Linear}})) {layerGroup=None}
  ]}
 Parachute {
  entry=Closed
  path=AnimGraph.AnimGraphNode_StateMachine_0.Parachute
  ids=id
  nodes [
   AnimStateNode_0 = Parachute_Into() {is=title AnimStateNode}
   AnimStateNode_1 = Idle()
   AnimStateNode_2 = Closed() {is=title AnimStateNode}
   AnimStateNode_3 = ForceOpen_or_GliderItem() {is=title AnimStateNode}
   AnimStateTransitionNode_0 = AnimStateTransitionNode() {bAutomaticRuleBasedOnSequencePlayerInState=true, crossfadeDuration=0}
   AnimStateTransitionNode_1 = AnimStateTransitionNode()
   AnimStateTransitionNode_2 = AnimStateTransitionNode() {crossfadeDuration=0}
   AnimStateTransitionNode_3 = AnimStateTransitionNode() {crossfadeDuration=0}
   AnimStateTransitionNode_4 = AnimStateTransitionNode() {crossfadeDuration=0}
   AnimStateTransitionNode_5 = AnimStateTransitionNode()
   AnimStateTransitionNode_6 = AnimStateTransitionNode() {bAutomaticRuleBasedOnSequencePlayerInState=true, crossfadeDuration=0}
  ]
  links [
   {
    dst=ForceOpen_or_GliderItem
    rule=AnimGraph.AnimGraphNode_StateMachine_0.Parachute.AnimStateTransitionNode_4.Transition
    src=Closed
    via=AnimStateTransitionNode_4
   }
   {
    dst=Parachute_Into
    rule=AnimGraph.AnimGraphNode_StateMachine_0.Parachute.AnimStateTransitionNode_1.Transition
    src=Closed
    via=AnimStateTransitionNode_1
   }
   {
    dst=Closed
    rule=AnimGraph.AnimGraphNode_StateMachine_0.Parachute.AnimStateTransitionNode_5.Transition
    src=ForceOpen_or_GliderItem
    via=AnimStateTransitionNode_5
   }
   {
    dst=Idle
    rule=AnimGraph.AnimGraphNode_StateMachine_0.Parachute.AnimStateTransitionNode_6.Transition
    src=ForceOpen_or_GliderItem
    via=AnimStateTransitionNode_6
   }
   {
    dst=Closed
    rule=AnimGraph.AnimGraphNode_StateMachine_0.Parachute.AnimStateTransitionNode_2.Transition
    src=Idle
    via=AnimStateTransitionNode_2
   }
   {
    dst=Closed
    rule=AnimGraph.AnimGraphNode_StateMachine_0.Parachute.AnimStateTransitionNode_3.Transition
    src=Parachute_Into
    via=AnimStateTransitionNode_3
   }
   {
    dst=Idle
    rule=AnimGraph.AnimGraphNode_StateMachine_0.Parachute.AnimStateTransitionNode_0.Transition
    src=Parachute_Into
    via=AnimStateTransitionNode_0
   }
  ]
 }
 Parachute_Into {
  path=AnimGraph.AnimGraphNode_StateMachine_0.Parachute.AnimStateNode_0.Parachute_Into
  nodes [
   n1 = Sequence Player 'Umbrella_HighTower_Into_FrontEnd_GLIDER'() {
    bLoopAnimation=false
    sequence=/BRCosmetics/Animation/Game/MainPlayer/Skydive/ParaGlide/Umbrella_HighTower/Umbrella_HighTower_Into_FrontEnd_GLIDER
   }
   n2 = Sequence Player 'Umbrella_HighTower_Into_GLIDER'!(PlayRate:float=1) -> (Pose) {bLoopAnimation=false, sequence=/BRCosmetics/Animation/Game/MainPlayer/Skydive/ParaGlide/Umbrella_HighTower/Umbrella_HighTower_Into_GLIDER}
   n3 = Get OwnerGlider() -> (OwnerGlider:Fort Player Parachute Object Reference)
   Output Animation Pose(Result=Blend Poses by bool(
     BlendPose_0=@n1
     BlendPose_1=@n2
     bActiveValue=Get IsCosmeticPreview!(self:Fort Player Parachute Object Reference=@n3) -> (IsCosmeticPreview:bool)
    ) {blendType=Linear})
  ]
 }
 Idle {
  path=AnimGraph.AnimGraphNode_StateMachine_0.Parachute.AnimStateNode_1.Idle
  nodes [
   n1 = Get OwnerGlider()
   Output Animation Pose(Result=Blend Poses by bool(
     BlendPose_0=Sequence Player 'Umbrella_Glide_FrontEnd_GLIDER'() {sequence=&/Umbrella_Glide_FrontEnd_GLIDER}
     BlendPose_1=Sequence Player 'Umbrella_HighTower_GLIDER'!(PlayRate:float=1) -> (Pose) {sequence=/BRCosmetics/Animation/Game/MainPlayer/Skydive/ParaGlide/Umbrella_HighTower/Umbrella_HighTower_GLIDER}
     bActiveValue=Get IsCosmeticPreview(self=@n1)
    ) {blendType=Linear})
  ]
 }
 Closed {
  path=AnimGraph.AnimGraphNode_StateMachine_0.Parachute.AnimStateNode_2.Closed
  nodes [
   n1 = Get OwnerGlider()
   Output Animation Pose(Result=Blend Poses by bool(
     BlendPose_0=Sequence Player 'Umbrella_Closed_FrontEnd_GLIDER'() {sequence=&/Umbrella_Closed_FrontEnd_GLIDER}
     BlendPose_1=Sequence Player 'Umbrella_HighTower_Closed_GLIDER'() {sequence=/BRCosmetics/Animation/Game/MainPlayer/Skydive/ParaGlide/Umbrella_HighTower/Umbrella_HighTower_Closed_GLIDER}
     bActiveValue=Get IsCosmeticPreview(self=@n1)
    ) {blendType=Linear})
  ]
 }
 ForceOpen_or_GliderItem {
  path=AnimGraph.AnimGraphNode_StateMachine_0.Parachute.AnimStateNode_3.ForceOpen_or_GliderItem
  nodes [Output Animation Pose(Result=Sequence Player 'Umbrella_Into_GLIDER'!(PlayRate:float=1) -> (Pose) {bLoopAnimation=false, sequence=&/Umbrella_Into_GLIDER})]
 }
 Transition {
  path=AnimGraph.AnimGraphNode_StateMachine_0.Parachute.AnimStateTransitionNode_0.Transition
  nodes [Result(bCanEnterTransition="Equal (Float)"!(A:double="Get Relevant Anim Time Remaining (Parachute_Into)"!() -> (ReturnValue:float) {is=title K2Node_AnimGetter}, B:double=0) -> (ReturnValue:bool) {is=title K2Node_CallFunction})]
 }
 Transition {
  path=AnimGraph.AnimGraphNode_StateMachine_0.Parachute.AnimStateTransitionNode_1.Transition
  nodes [
   n1 = Get SkydivingState() -> (SkydivingState:Fort Anim Input Skydiving Structure)
   Result(bCanEnterTransition=OR Boolean(
     A=Break Fort Anim Input Skydiving!(FortAnimInput_Skydiving:"Fort Anim Input Skydiving Structure (by ref)"=@n1) -> (
      ParaGilderRootOffset:Vector=~
      ParaGliderRootRotationOverride:Rotator=~
      bUseParaGlideRootModifier:bool=~
      bIsSkydivingFromLaunchPad:bool=~
      bIsSkydivingFromBus:bool=~
      bIsInVortex:bool=~
      bIsDiving:bool=~
      bIsParachuteOpen:bool=~
      bIsSkydiving:bool=~
      bIsParachuteLeaning:bool=~
      bIsSkydiveLeaning:bool=~
      bIsLeaning:bool=~
      bIsSkydiveDiveMode:bool=~
      bPlaySkydiveDrift:bool=~
      bSkydiveDriftDelayActive:bool=~
      bSkydiveDriftAnimAllowed:bool=~
      bIsGliderRight:bool=~
      bIsGliderCenter:bool=~
      bIsGliderLeft:bool=~
      bIsGliderForward:bool=~
      bIsGliderBack:bool=~
      bHasDeployedAndIsStandingUpright:bool=~
      LocalAccelForward:float=~
      LocalAccelRight:float=~
      LocalVelocityRight:float=~
      SkydiveAimPitch:float=~
      SkydiveAimYaw:float=~
      DeployChuteAnimRate:float=~
      SkydiveDriftAnimRate:float=~
      SkydiveDriftAnimRateCurrent:float=~
      SkydiveFidgetAnimRate:float=~
      SkydiveFidgetAnimRateCurrent:float=~
      SkydiveAdditiveAlpha:float=~
      SkydiveDriftDelay:float=~
      SkydiveDriftAnim:int
      SkydiveDriftAnimMax:int
      LaunchpadAnim:int
      LocalAccelDir:ESkydivingDirection Enum
      DirectionLast:ESkydivingDirection Enum
      PoseAdditive:Anim Sequence Object Reference
     ) {is=title K2Node_BreakStruct}.bIsParachuteOpen
     B=Get bForceOpen() -> (bForceOpen:bool)
    ))
   n2 = Break Fort Anim Input Skydiving(FortAnimInput_Skydiving=Get SkydivingState()) -> (
    ParaGilderRootOffset=~
    ParaGliderRootRotationOverride=~
    bUseParaGlideRootModifier=~
    bIsSkydivingFromLaunchPad=~
    bIsSkydivingFromBus=~
    bIsInVortex=~
    bIsDiving=~
    bIsParachuteOpen=~
    bIsSkydiving=~
    bIsParachuteLeaning=~
    bIsSkydiveLeaning=~
    bIsLeaning=~
    bIsSkydiveDiveMode=~
    bPlaySkydiveDrift=~
    bSkydiveDriftDelayActive=~
    bSkydiveDriftAnimAllowed=~
    bIsGliderRight=~
    bIsGliderCenter=~
    bIsGliderLeft=~
    bIsGliderForward=~
    bIsGliderBack=~
    bHasDeployedAndIsStandingUpright=~
    LocalAccelForward=~
    LocalAccelRight=~
    LocalVelocityRight=~
    SkydiveAimPitch=~
    SkydiveAimYaw=~
    DeployChuteAnimRate=~
    SkydiveDriftAnimRate=~
    SkydiveDriftAnimRateCurrent=~
    SkydiveFidgetAnimRate=~
    SkydiveFidgetAnimRateCurrent=~
    SkydiveAdditiveAlpha=~
    SkydiveDriftDelay=~
   )
   n3 = Get OwnerGlider()
   AND Boolean(
    A=OR Boolean!(A=@n2.bIsSkydivingFromLaunchPad, B=@n2.bIsSkydivingFromBus, C:bool=@n2.bIsInVortex, D:bool=Get IsCosmeticPreview(self=@n3)) -> (ReturnValue)
    B=@n2.bIsParachuteOpen
   )
  ]
 }
 Transition {
  path=AnimGraph.AnimGraphNode_StateMachine_0.Parachute.AnimStateTransitionNode_2.Transition
  nodes [Result(bCanEnterTransition=NOT Boolean(A=Get IsParachuteOpen() -> (IsParachuteOpen:bool)))]
 }
 Transition {
  path=AnimGraph.AnimGraphNode_StateMachine_0.Parachute.AnimStateTransitionNode_3.Transition
  nodes [
   n1 = Break Fort Anim Input Skydiving(FortAnimInput_Skydiving=Get SkydivingState()) -> (
    ParaGilderRootOffset=~
    ParaGliderRootRotationOverride=~
    bUseParaGlideRootModifier=~
    bIsSkydivingFromLaunchPad=~
    bIsSkydivingFromBus=~
    bIsInVortex=~
    bIsDiving=~
    bIsParachuteOpen=~
    bIsSkydiving=~
    bIsParachuteLeaning=~
    bIsSkydiveLeaning=~
    bIsLeaning=~
    bIsSkydiveDiveMode=~
    bPlaySkydiveDrift=~
    bSkydiveDriftDelayActive=~
    bSkydiveDriftAnimAllowed=~
    bIsGliderRight=~
    bIsGliderCenter=~
    bIsGliderLeft=~
    bIsGliderForward=~
    bIsGliderBack=~
    bHasDeployedAndIsStandingUpright=~
    LocalAccelForward=~
    LocalAccelRight=~
    LocalVelocityRight=~
    SkydiveAimPitch=~
    SkydiveAimYaw=~
    DeployChuteAnimRate=~
    SkydiveDriftAnimRate=~
    SkydiveDriftAnimRateCurrent=~
    SkydiveFidgetAnimRate=~
    SkydiveFidgetAnimRateCurrent=~
    SkydiveAdditiveAlpha=~
    SkydiveDriftDelay=~
   )
   n2 = Get bForceOpen()
   Result(bCanEnterTransition=AND Boolean(A=NOT Boolean(A=@n1.bIsParachuteOpen), B=NOT Boolean(A=@n2)))
  ]
 }
 Transition {
  path=AnimGraph.AnimGraphNode_StateMachine_0.Parachute.AnimStateTransitionNode_4.Transition
  nodes [
   Result(bCanEnterTransition=Get BlockAnimRule() -> (BlockAnimRule:bool))
   n1 = Break Fort Anim Input Skydiving(FortAnimInput_Skydiving=Get SkydivingState()) -> (
    ParaGilderRootOffset=~
    ParaGliderRootRotationOverride=~
    bUseParaGlideRootModifier=~
    bIsSkydivingFromLaunchPad=~
    bIsSkydivingFromBus=~
    bIsInVortex=~
    bIsDiving=~
    bIsParachuteOpen=~
    bIsSkydiving=~
    bIsParachuteLeaning=~
    bIsSkydiveLeaning=~
    bIsLeaning=~
    bIsSkydiveDiveMode=~
    bPlaySkydiveDrift=~
    bSkydiveDriftDelayActive=~
    bSkydiveDriftAnimAllowed=~
    bIsGliderRight=~
    bIsGliderCenter=~
    bIsGliderLeft=~
    bIsGliderForward=~
    bIsGliderBack=~
    bHasDeployedAndIsStandingUpright=~
    LocalAccelForward=~
    LocalAccelRight=~
    LocalVelocityRight=~
    SkydiveAimPitch=~
    SkydiveAimYaw=~
    DeployChuteAnimRate=~
    SkydiveDriftAnimRate=~
    SkydiveDriftAnimRateCurrent=~
    SkydiveFidgetAnimRate=~
    SkydiveFidgetAnimRateCurrent=~
    SkydiveAdditiveAlpha=~
    SkydiveDriftDelay=~
   )
   AND Boolean!(
    A=NOT Boolean(A=@n1.bIsSkydivingFromLaunchPad)
    B=NOT Boolean(A=@n1.bIsSkydivingFromBus)
    C:bool=NOT Boolean(A=@n1.bIsInVortex)
    D:bool=Get bForceOpen()
   ) -> (ReturnValue)
  ]
 }
 Transition {
  path=AnimGraph.AnimGraphNode_StateMachine_0.Parachute.AnimStateTransitionNode_5.Transition
  nodes [
   n1 = Break Fort Anim Input Skydiving(FortAnimInput_Skydiving=Get SkydivingState()) -> (
    ParaGilderRootOffset=~
    ParaGliderRootRotationOverride=~
    bUseParaGlideRootModifier=~
    bIsSkydivingFromLaunchPad=~
    bIsSkydivingFromBus=~
    bIsInVortex=~
    bIsDiving=~
    bIsParachuteOpen=~
    bIsSkydiving=~
    bIsParachuteLeaning=~
    bIsSkydiveLeaning=~
    bIsLeaning=~
    bIsSkydiveDiveMode=~
    bPlaySkydiveDrift=~
    bSkydiveDriftDelayActive=~
    bSkydiveDriftAnimAllowed=~
    bIsGliderRight=~
    bIsGliderCenter=~
    bIsGliderLeft=~
    bIsGliderForward=~
    bIsGliderBack=~
    bHasDeployedAndIsStandingUpright=~
    LocalAccelForward=~
    LocalAccelRight=~
    LocalVelocityRight=~
    SkydiveAimPitch=~
    SkydiveAimYaw=~
    DeployChuteAnimRate=~
    SkydiveDriftAnimRate=~
    SkydiveDriftAnimRateCurrent=~
    SkydiveFidgetAnimRate=~
    SkydiveFidgetAnimRateCurrent=~
    SkydiveAdditiveAlpha=~
    SkydiveDriftDelay=~
   )
   n2 = Get bForceOpen()
   Result(bCanEnterTransition=AND Boolean!(A=NOT Boolean(A=@n1.bIsParachuteOpen), B=NOT Boolean(A=@n2)) -> (ReturnValue))
  ]
 }
 Transition {path=AnimGraph.AnimGraphNode_StateMachine_0.Parachute.AnimStateTransitionNode_6.Transition, nodes [Result()]}
 EventGraph {nodes [
   n1 = Event BlueprintUpdateAnimation()
   > Set Time(Time:double=@n1.DeltaTimeX) -> (Output_Get:double)
   > Get Player Open Glider Bool!(execute:Exec) -> (then:Exec) {is=title K2Node_MacroInstance}
   n2 = Break Fort Anim Input Skydiving(FortAnimInput_Skydiving=Get SkydivingState()) -> (
    ParaGilderRootOffset=~
    ParaGliderRootRotationOverride=~
    bUseParaGlideRootModifier=~
    bIsSkydivingFromLaunchPad=~
    bIsSkydivingFromBus=~
    bIsInVortex=~
    bIsDiving=~
    bIsParachuteOpen=~
    bIsSkydiving=~
    bIsParachuteLeaning=~
    bIsSkydiveLeaning=~
    bIsLeaning=~
    bIsSkydiveDiveMode=~
    bPlaySkydiveDrift=~
    bSkydiveDriftDelayActive=~
    bSkydiveDriftAnimAllowed=~
    bIsGliderRight=~
    bIsGliderCenter=~
    bIsGliderLeft=~
    bIsGliderForward=~
    bIsGliderBack=~
    bHasDeployedAndIsStandingUpright=~
    LocalAccelForward=~
    LocalAccelRight=~
    LocalVelocityRight=~
    SkydiveAimPitch=~
    SkydiveAimYaw=~
    DeployChuteAnimRate=~
    SkydiveDriftAnimRate=~
    SkydiveDriftAnimRateCurrent=~
    SkydiveFidgetAnimRate=~
    SkydiveFidgetAnimRateCurrent=~
    SkydiveAdditiveAlpha=~
    SkydiveDriftDelay=~
   )
   > Set IsParachuteOpen(IsParachuteOpen:bool=@n2.bIsParachuteOpen) -> (Output_Get:bool)
   > Set DeployAnimRate(DeployAnimRate:double=@n2.DeployChuteAnimRate) -> (Output_Get:double)
   > n3 = Branch(Condition=Get IsCosmeticPreview(self=Get OwnerGlider()))
   > Set TopSpin_Speed(TopSpin_Speed:double=0) -> (Output_Get:double)
   n4 = ForwardBack_Value_Interp!(
    execute=@n3.else
    Time:double=Get Time() -> (Time:double)
    Currnet_Value:double=Get TopSpin_Speed()
    Idle_Value:double=1
    Idle_InterpSpeed:double=1
    Forward_Value:double=3
    Forward_InterpSpeed:double=2
    Back_Value:double=-1.75
    Back_InterpSpeed:double=2
    LocalAccelForward:double=@n2.LocalAccelForward
   ) -> (then, New_Value:double)
   > Set TopSpin_Speed(TopSpin_Speed=@n4)
   AnimNotify_TurnONEngines() {is=title K2Node_Event}
   > SetGliderFullyDeployed!(execute, self:Fort Glider Anim Instance Object Reference, bIsFullyDeployed:bool=true) -> (then)
  ]}
 GetPlayerOpenGliderBool {nodes [
   Inputs()
   > n1 = Branch(Condition=Get IsCosmeticPreview(self=Get OwnerGlider()))
   Outputs!(then:Exec=[@n1.then, @n3.then])
   n2 = Get OwnerFortPawn() -> (OwnerFortPawn:Fort Player Pawn Object Reference)
   n3 = Set bForceOpen(
    execute=@n1.else
    bForceOpen:bool=ShouldForceOpenGlider!(self:Fort Player Animation Interface Object Reference=GetAnimationInterface!(self:Fort Player Pawn Object Reference=@n2) -> (ReturnValue:Fort Player Animation Interface Object Reference)) -> (ReturnValue:bool)
   ) -> (Output_Get:bool)
  ]}
]
variables [
 BlockAnimRule {type=boolean}
 DeployAnimRate {type=number}
 IsParachuteOpen {type=boolean}
 SideSpinner_Speed {type=number, default=1}
 Time {type=number}
 TopSpin_Speed {type=number, default=1}
]
