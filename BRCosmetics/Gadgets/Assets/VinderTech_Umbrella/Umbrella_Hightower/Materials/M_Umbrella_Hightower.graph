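# Character-style material for the VinderTech Hightower umbrella gadget, adding a masked, time-of-day emissive between the shared character PreFX and PostFX functions.
# The E texture masks both BaseColorDarkening and the glow, which blends day and night colors through Color_Time_of_Day.
# A SphereMask on the inverted F texture gates the glow by Radius and Hardness, acting as a gradient threshold.
# A scaled, offset pre-skinned-position VertexInterpolator chain is left unconnected.
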Material /BRCosmetics/Gadgets/Assets/VinderTech_Umbrella/Umbrella_Hightower/Materials/M_Umbrella_Hightower
graph {
 properties {
  bUseMaterialAttributes=true
  bUsedWithMeshDeformer=true
  bUsedWithSkeletalMesh=true
  pixelDepthOffsetMode=PDOM_Legacy
  usedShadingModels=Default Lit
 }
 nodes [
  ComponentMask() {g=true, r=true}
  n1 = MaterialFunctionCall() {materialFunction=&BaseFN_CharacterFunction_PreFX}
  n2 = GetMaterialAttributes(None=@n1.Result) {attributeGetTypes [#1]}
  n3 = TextureSample() {
   samplerType=SAMPLERTYPE_LinearGrayscale
   texture=/BRCosmetics/Gadgets/Assets/VinderTech_Umbrella/Umbrella_Hightower/Textures/T_Umbrella_Hightower_E_TA
  }
  n4 = Max(A=OneMinus(None=@n3.R), B=ScalarParameter() {group=Hightower, parameterName=BaseColorDarkening})
  n5 = TextureSample() {
   samplerType=SAMPLERTYPE_LinearGrayscale
   texture=/BRCosmetics/Gadgets/Assets/VinderTech_Umbrella/Umbrella_Hightower/Textures/T_Umbrella_Hightower_F
  }
  n6 = OneMinus(None=@n5.R)
  n7 = Constant()
  n8 = ScalarParameter() {defaultValue=100, group=Hightower, parameterName=Radius}
  n9 = ScalarParameter() {defaultValue=0.99, group=Hightower, parameterName=Hardness}
  n10 = Multiply(A=@n3.R, B=OneMinus(None=SphereMask(A=@n6, B=@n7, Radius=@n8, Hardness=@n9))) {constA=1}
  n11 = VectorParameter() {group=Hightower, parameterName=EmissiveColorDay}
  n12 = MaterialFunctionCall(
   Color_Morning=@n11.RGB
   Color_Day=@n11.RGB
   Color_Evening=@n11.RGB
   Color_Night=VectorParameter() {group=Hightower, parameterName=EmissiveColorNight}.RGB
  ) {materialFunction=&Color_Time_of_Day}
  MaterialFunctionCall(
   AO=@n1.AO
   Distance Blend=@n1.AO
   Fresnel=@n1.AO
   Material=SetMaterialAttributes!(MaterialAttributes=@n2.MaterialAttributes, Base Color=Multiply(A=@n2.Base Color, B=@n4), Emissive Color=Multiply(A=@n10, B=@n12.Result)) {attributeSetTypes [#1, #2]}
  ) {materialFunction=&BaseFN_CharacterFunction_PostFX}
  n13 = ComponentMask(None=PreSkinnedPosition()) {g=true, r=true}
  n14 = ComponentMask(None=VectorParameter() {group=Hightower, parameterName=Scale, defaultValue=<rgba 1 1 0 0>}.RGB) {g=true, r=true}
  n15 = VectorParameter() {group=Hightower, parameterName=Offset}
  VertexInterpolator(VS=Add(A=Multiply(A=@n13, B=@n14), B=ComponentMask(None=@n15.RGB) {g=true, r=true}))
 ]
}
